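# Conversation database for the Sunbird NPC, giving its greeting dialogue when a player starts talking to it.
# The entry point leads to one speech node with a default line plus contextual lines for first-ever conversations and for players wearing specific character cosmetics.
# The first-meeting line takes its priority from a curve row, while the cosmetic-specific lines use a fixed priority.
# The speech node then links out to the shared NPC Library quests and basic services conversations.
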ConversationDatabase /NPCLibrary/NPCs/Sunbird/Sunbird/NPCLibrary_Conversation_Sunbird
entry_points [{tag=Conversation.Entry.Sunbird, destinations [#1]}]
nodes [
 ConversationEntryPointNode_7 {
  id=#1
  class=ConversationEntryPointNode
  outputs [#2]
  properties {entryTag {tagName=Conversation.Entry.Sunbird}}
 }
 Speech_C_4 {
  id=#2
  class=Speech_C
  outputs [#3, #4]
  properties {
   nodeName=""
   generalConfig {
    defaultMessage="I count and I count, and there are fewer stars each time."
    contextualMessages [
     {
      message="If our fate is in the stars, what happens when they are gone?"
      contextRequirements [{requirement {class=FortControllerRequirement_IsFirstEverConversationWithNPC}}]
      priority {curve {rowName=Tandem.Speech.Priority.FirstEverInteraction}}
     }
     {
      message=Great power comes at great cost.
      contextRequirements [{requirement {class=FortControllerRequirement_HasCID, properties {allowedCIds [&Character_Bites]}}}]
      priority {value=0, curve {curveTable=null, rowName=None}}
     }
     {
      message="You want to carve your own path, don't you?"
      contextRequirements [{requirement {class=FortControllerRequirement_HasCID, properties {allowedCIds [&Character_ChillCat]}}}]
      priority {value=0, curve {curveTable=null, rowName=None}}
     }
     {
      message=You have your freedom. What will you do with it?
      contextRequirements [{requirement {class=FortControllerRequirement_HasCID, properties {allowedCIds [&Character_PinkSpike]}}}]
      priority {value=0, curve {curveTable=null, rowName=None}}
     }
     {
      message="Have you walked away from your true nature, or embraced it?"
      contextRequirements [{requirement {class=FortControllerRequirement_HasCID, properties {allowedCIds [&Character_BadBear]}}}]
      priority {value=0, curve {curveTable=null, rowName=None}}
     }
     {
      message="In the end, what you herald will consume you."
      contextRequirements [{requirement {class=FortControllerRequirement_HasCID, properties {allowedCIds [&Character_TheHerald]}}}]
      priority {value=0, curve {curveTable=null, rowName=None}}
     }
     {
      message=You and me. The spider and the bird.
      contextRequirements [{requirement {class=FortControllerRequirement_HasCID, properties {allowedCIds [&Character_RoseDust]}}}]
      priority {value=0, curve {curveTable=null, rowName=None}}
     }
    ]
   }
  }
 }
 ConversationLinkNode_0 {id=#4, class=ConversationLinkNode, properties {remoteEntryTag {tagName=Conversation.Entry.BasicServices}}}
 ConversationLinkNode_1 {id=#3, class=ConversationLinkNode, properties {remoteEntryTag {tagName=Conversation.Entry.NPCLibrary.Quests}}}
]
